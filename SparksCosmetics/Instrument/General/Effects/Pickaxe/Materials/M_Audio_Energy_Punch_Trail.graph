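# Translucent, unlit, two-sided material for Niagara mesh particles, drawing the energy punch trail of an audio-themed instrument pickaxe effect.
# Offset, panning smoke-texture samples fill separate RGB channels for chromatic fringing, with scrolling noise distorting and masking them.
# Edge fades shape the trail, color blends particle color, chroma and an optional two-color gradient, and opacity uses non-mobile depth fade.
# SplineThicken drives vertex offset, with optional dynamic-parameter width scaling and noise wobble.
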Material /SparksCosmetics/Instrument/General/Effects/Pickaxe/Materials/M_Audio_Energy_Punch_Trail
graph {
 properties {
  bAutomaticallySetUsageInEditor=false
  bUsedWithNiagaraMeshParticles=true
  bUsedWithStaticMesh=false
  blendMode=BLEND_Translucent
  pixelDepthOffsetMode=PDOM_Legacy
  shadingModel=MSM_Unlit
  twoSided=true
  usedShadingModels=Unlit
 }
 nodes [
  DynamicParameter() {defaultValue=<rgba 0 1 1 1>, paramNames [Dissolve, Panner offset, WPO, Width_Scale]}
  n1 = Constant() {r=0.03}
  n2 = TextureCoordinate() {vTiling=0.5}
  n3 = Multiply(A=ComponentMask(None=@n2) {g=true}, B=ScalarParameter() {defaultValue=1, parameterName=Texture_UVScale})
  n4 = ComponentMask(None=@n2) {r=true}
  n5 = ScalarParameter() {defaultValue=1, parameterName=Overall_ScaleX}
  n6 = ScalarParameter() {defaultValue=1, parameterName=Overall_ScaleY}
  n7 = ParticleRandom()
  n8 = Constant()
  n9 = OneMinus(None=@n6)
  n10 = Add(
   A=Multiply(
    A=MaterialFunctionCall(X=@n3, Y=@n4) {materialFunction=&MakeFloat2}.Result
    B=AppendVector(A=@n5, B=@n6)
   )
   B=AppendVector(A=StaticSwitchParameter(True=@n7, False=@n8) {parameterName=UseParticleRandom}, B=Multiply(A=@n9) {constB=0.5})
  )
  n11 = AppendVector(A=ScalarParameter() {defaultValue=-0.3, parameterName=Texture_A_Panner_speed}, B=Constant())
  n12 = TextureSample(UVs=Panner(Coordinate=Add(A=@n10, B=@n1) {constB=0.05}, Speed=@n11) {speedX=-0.3}) {
   samplerSource=SSM_Wrap_WorldGroupSettings
   samplerType=SAMPLERTYPE_Grayscale
   texture=&T_Smoke_Trail_Soft
  }
  n13 = TextureSample(UVs=Panner(Coordinate=@n10, Speed=@n11) {speedX=-0.3}) {
   samplerSource=SSM_Wrap_WorldGroupSettings
   samplerType=SAMPLERTYPE_Grayscale
   texture=&T_Smoke_Trail_Soft
  }
  n14 = Multiply(A=@n1) {constB=-1}
  n15 = TextureSample(UVs=Panner(Coordinate=Add(A=@n10, B=@n14) {constB=-0.05}, Speed=@n11) {speedX=-0.3}) {
   samplerSource=SSM_Wrap_WorldGroupSettings
   samplerType=SAMPLERTYPE_Grayscale
   texture=&T_Smoke_Trail_Soft
  }
  n16 = Multiply(A=Power(Base=MaterialFunctionCall(X=@n12.R, Y=@n13.G, Z=@n15.B) {materialFunction=&MakeFloat3}.Result)) {constB=4}
  n17 = Constant() {r=0.03}
  n18 = TextureCoordinate() {vTiling=0.5}
  n19 = Multiply(A=ComponentMask(None=@n18) {g=true}, B=ScalarParameter() {defaultValue=1, parameterName=Texture_UVScale2})
  n20 = ComponentMask(None=@n18) {r=true}
  n21 = ScalarParameter() {defaultValue=1, parameterName=Overall_ScaleX}
  n22 = ScalarParameter() {defaultValue=1, parameterName=Overall_ScaleY}
  n23 = ParticleRandom()
  n24 = Constant()
  n25 = OneMinus(None=@n22)
  n26 = Add(
   A=Multiply(
    A=MaterialFunctionCall(X=@n19, Y=@n20) {materialFunction=&MakeFloat2}.Result
    B=AppendVector(A=@n21, B=@n22)
   )
   B=AppendVector(A=StaticSwitchParameter(True=@n23, False=@n24) {parameterName=UseParticleRandom}, B=Multiply(A=@n25) {constB=0.5})
  )
  n27 = Add(A=@n26, B=@n17) {constB=0.05}
  n28 = AppendVector(A=ScalarParameter() {defaultValue=-0.3, parameterName=Texture_A_Panner_speed}, B=Constant())
  n29 = TextureCoordinate()
  n30 = ComponentMask(None=@n29) {g=true}
  n31 = ComponentMask(None=@n29) {r=true}
  n32 = AppendVector(A=ScalarParameter() {defaultValue=0.5, parameterName=Texture_B_Panner_speed}, B=Constant())
  n33 = TextureSample(UVs=Panner(
    Coordinate=MaterialFunctionCall(X=@n30, Y=@n31) {materialFunction=&MakeFloat2}.Result
    Speed=@n32
   ) {speedX=0.5}) {samplerType=SAMPLERTYPE_Grayscale, texture=&LowResBlurredNoise}
  n34 = TextureCoordinate()
  n35 = ComponentMask(None=@n34) {g=true}
  n36 = ScalarParameter() {defaultValue=1, parameterName=Texture_UVScale}
  n37 = ScalarParameter() {defaultValue=1, parameterName=Overall_ScaleY}
  n38 = Multiply(
   A=MaterialFunctionCall(X=Multiply(A=@n35, B=@n36), Y=ComponentMask(None=@n34) {r=true}) {materialFunction=&MakeFloat2}.Result
   B=AppendVector(A=ScalarParameter() {defaultValue=1, parameterName=Overall_ScaleX}, B=@n37)
  )
  n39 = AppendVector(
   A=StaticSwitchParameter(True=ParticleRandom(), False=Constant()) {parameterName=UseParticleRandom}
   B=Multiply(A=OneMinus(None=@n37)) {constB=0.5}
  )
  n40 = TextureSample(UVs=Panner(Coordinate=Add(A=@n38, B=@n39), Speed=@n32) {speedX=0.5}) {
   samplerSource=SSM_Wrap_WorldGroupSettings
   samplerType=SAMPLERTYPE_Grayscale
   texture=&T_Smoke_Trail_Soft
  }
  n41 = Saturate(None=Multiply(A=Multiply(A=@n33.R, B=@n40.R)) {constB=2})
  n42 = Multiply(A=@n41) {constB=0.1}
  n43 = TextureSample(UVs=Add(A=Panner(Coordinate=@n27, Speed=@n28) {speedX=-0.4}, B=@n42)) {
   samplerSource=SSM_Wrap_WorldGroupSettings
   samplerType=SAMPLERTYPE_Grayscale
   texture=&T_Smoke_Trail_Soft
  }
  n44 = TextureSample(UVs=Add(A=Panner(Coordinate=@n26, Speed=@n28) {speedX=-0.3}, B=@n42)) {
   samplerSource=SSM_Wrap_WorldGroupSettings
   samplerType=SAMPLERTYPE_Grayscale
   texture=&T_Smoke_Trail_Soft
  }
  n45 = Add(A=@n26, B=Multiply(A=@n17) {constB=-1}) {constB=-0.05}
  n46 = TextureSample(UVs=Add(A=Panner(Coordinate=@n45, Speed=@n28) {speedX=-0.4}, B=@n42)) {
   samplerSource=SSM_Wrap_WorldGroupSettings
   samplerType=SAMPLERTYPE_Grayscale
   texture=&T_Smoke_Trail_Soft
  }
  n47 = MaterialFunctionCall(X=@n43.R, Y=@n44.G, Z=@n46.B) {materialFunction=&MakeFloat3}
  LinearInterpolate(A=@n16, B=@n47.Result)
  n48 = ParticleColor()
  n49 = Multiply(A=Multiply(A=@n16, B=@n41), B=@n47.Result)
  n50 = OneMinus(None=ComponentMask(None=TextureCoordinate()) {g=true})
  n51 = ScalarParameter() {defaultValue=2, parameterName=TopBottom_Fade}
  n52 = TextureCoordinate()
  n53 = Multiply(A=Multiply(A=Multiply(A=@n49) {constB=50}, B=Power(Base=@n50, Exp=@n51)), B=Power(Base=ComponentMask(None=@n52) {g=true}, Exp=@n51))
  n54 = Multiply(A=Desaturation(None=@n53, Fraction=Constant() {r=-0.25})) {constB=1.5}
  n55 = ScalarParameter() {defaultValue=1, parameterName=Use_Chroma_Lerp}
  n56 = Constant() {r=1}
  n57 = Desaturation(None=@n53)
  n58 = LinearInterpolate(
   A=VectorParameter() {parameterName=Lerp_Color2, defaultValue=<rgba 1 0.0976 0 1>}.RGB
   B=VectorParameter() {parameterName=Lerp_Color1, defaultValue=<rgba 1 0.6604 0.1499 1>}.RGB
   Alpha=@n57
  )
  n59 = ScalarParameter() {parameterName=Use_Lerp_Color}
  MaterialFunctionCall(Emissive=Multiply(A=LinearInterpolate(A=@n48.RGB, B=@n54, Alpha=@n55), B=LinearInterpolate(A=@n56, B=@n58, Alpha=@n59))) {materialFunction=&MF_FX_EmissiveAdjustments}
  Max(A=Max(A=@n43.A, B=@n44.A), B=@n46.A)
  n60 = Multiply(A=@n48.A, B=Multiply(A=@n57, B=@n41))
  n61 = ScalarParameter() {defaultValue=70, parameterName=Depth_Fade}
  Multiply(A=ShadingPathSwitch(Default=DepthFade(Opacity=@n60, FadeDistance=@n61), Mobile=@n60) {desc=Migrated from FeatureLevelSwitch})
  n62 = Panner(Coordinate=TextureCoordinate()) {speedY=0.5}
  n63 = Multiply(
   A=Transform(None=TextureSample(UVs=@n62) {samplerType=SAMPLERTYPE_LinearColor, texture=&T_NoisePack_LinearColor}.RGB)
   B=DynamicParameter() {defaultValue=<rgba 0 1 1 1>, paramNames [Dissolve, Panner offset, WPO, Width_Scale]}.WPO
  )
  n64 = ScalarParameter() {defaultValue=30, parameterName=Widthbase}
  n65 = DynamicParameter() {defaultValue=<rgba 0 1 1 1>, paramNames [Dissolve, Panner offset, WPO, Width_Scale]}
  n66 = ScalarParameter() {parameterName=Use_WPO_Width_Scale}
  n67 = ScalarParameter() {defaultValue=30, parameterName=WidthTip}
  n68 = MaterialFunctionCall(
   WidthBase=LinearInterpolate(A=@n64, B=Multiply(A=@n64, B=@n65.Width_Scale), Alpha=@n66)
   WidthTip=LinearInterpolate(A=@n67, B=Multiply(A=@n67, B=@n65.Width_Scale), Alpha=@n66)
  ) {materialFunction=&SplineThicken}
  StaticSwitchParameter(
   True=LinearInterpolate(A=@n68.WorldPositionOffset, B=Add(A=@n68.WorldPositionOffset, B=@n63), Alpha=ScalarParameter() {parameterName=Add_WPO_Noise})
   False=Constant()
  ) {defaultValue=true, parameterName=Use SplineThicken}
 ]
}
